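; Quest gameplay ability for Spooky Mist "approach player" that credits the owning player for hostile players nearby when the ability ends.
; Activation from an event binds the owner's DBNO-start dispatcher to cancel the ability; ending unbinds it.
; On any end, including cancellation, it counts hostile pawns within a meter-based radius and sends that count to the quest manager as a custom stat under the quest tag.
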(event Ability|EventActivateAbilityFromEvent
  (params (EventData :type Gameplay%20Event%20Data%20Structure%20%28by%20ref%29))
  ;[1 1 1] Bind to the DBNO event
  (|CallParentFunction EventData)
  (EventDispatchers|BindEventtoOnPlayerStartDBNO (Variables|GAVars|Data|GetOwningPlayerPawn) (EventDispatchers|CreateEvent :!event K2_CancelAbility)))

(event Custom|CheckForNearbyEnemies
  ;[1 1 1] Handle checking for nearby enemies
  (Variables|Default|SetNumberOfPlayersApproached 0)
  (bind _outfortplayerpawns (Fortnite|GetFortPlayerPawnsInRange (Transformation|GetActorLocation (Variables|GAVars|Data|GetOwningPlayerPawn)) (* (Variables|Default|GetRadiusInMeters) 1e+02)))
  ;;[1 1 1] Check each hit actor until we find an enemy
  (for _array_element _outfortplayerpawns
    (if (Utilities|Enum|Equal(Enum) (Team|GetTeamAffiliation _array_element (Variables|GAVars|Data|GetOwningPlayerPawn)) "Hostile")
      (Math|Integer|IncrementInt (Variables|Default|GetNumberOfPlayersApproached))))
  (Quest|SendComplexCustomStatEvent (Variables|GAVars|Data|GetOwningPlayerQuestManager) 0 (Variables|Default|GetQuestTag) "(GameplayTags=)" (Variables|Default|GetNumberOfPlayersApproached)))

(event Ability|EventOnEndAbility
  (params (bWasCancelled :type Boolean))
  ;[1 1 1] OnEndAbility: When the ability ends, check to see if any enemies are nearby and register their stat.
  (|CallParentFunction bWasCancelled)
  (CallFunction|CheckforNearbyEnemies)
  (EventDispatchers|UnbindEventfromOnPlayerStartDBNO (Variables|GAVars|Data|GetOwningPlayerPawn) (EventDispatchers|CreateEvent :!event K2_CancelAbility)))
